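# Physics asset for the Rust Lord Remix Scavenger back bling, a cosmetic backpack with a knot bone that has its own physics body.
# It defines two box bodies, one on the pack bone and a thinner, slightly offset one on dyn_knot.
# A single constraint joins dyn_knot to pack with angle limits set on both swing axes and a much tighter twist limit.
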PhysicsAsset /BRCosmetics/Accessories/FORT_Backpacks/M_MED_RustLordRemixScavenger_Backpack/Meshes/Meshes/M_MED_Rust_Lord_Remix_Scavenger_Pack_Physics
bodies {
 dyn_knot [{shapeType=Box, extentX=1.8184, extentY=8.5454, extentZ=8.3018, center=<-1 0 0>}]
 pack [{shapeType=Box, extentX=7.5, extentY=7.5, extentZ=7.5}]
}
constraints [{bone1Name=dyn_knot, bone2Name=pack, swing1LimitDegrees=35, swing2LimitDegrees=25, twistLimitDegrees=5}]
